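# Translucent effect material for the Kettle vehicle cosmetic wheel, drawing glowing stripes in an overridable FXColor, red by default.
# MF_WheelParent_Simple supplies base attributes, with emissive replaced by the boosted, slightly desaturated color.
# Opacity is procedural, without textures: tiled V coordinates form eleven thin, power-sharpened bands that fade along U.
# The result passes through the shared vehicle damage and de-rez transition functions.
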Material /VehicleCosmetics/Wheels/Kettle/Materials/M_Wheel_Kettle_FX
graph {
 properties {
  bAutomaticallySetUsageInEditor=false
  bUseMaterialAttributes=true
  bUsedWithMeshDeformer=true
  bUsedWithSkeletalMesh=true
  blendMode=BLEND_Translucent
  pixelDepthOffsetMode=PDOM_Legacy
  previewMesh=/VehicleCosmetics/Wheels/Sprinkle02/Meshes/SK_Wheel_Sprinkle02
  usedShadingModels=Default Lit
 }
 nodes [
  n1 = MaterialFunctionCall() {materialFunction=&MF_WheelParent_Simple}
  n2 = Desaturation(
   None=Multiply(A=VectorParameter() {group=5_Special Effects, parameterName=FXColor, defaultValue=<rgba 1 0.015 0.015 1>}.RGB, B=Constant() {r=16})
   Fraction=Constant() {r=0.2}
  )
  n3 = TextureCoordinate()
  n4 = Frac(None=Multiply(A=@n3, B=Constant2Vector() {g=11, r=1}))
  n5 = Constant() {r=-0.5}
  n6 = Abs(None=ComponentMask(None=Add(A=@n4, B=@n5)) {g=true})
  n7 = Multiply(A=OneMinus(None=ComponentMask(None=@n3) {r=true}), B=Power(Base=OneMinus(None=@n6), Exp=Constant() {r=15}))
  MaterialFunctionCall(In=MaterialFunctionCall(Material=SetMaterialAttributes!(MaterialAttributes=@n1.MaterialAttributes, Emissive Color=@n2, Opacity=@n7) {attributeSetTypes [#1, #2]}) {materialFunction=&MF_BaseVehicle_Damage}.Result) {materialFunction=&MF_VehicleDeRez_Transition}
 ]
}
